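# Post-copy animation Blueprint for the FalseVulture male medium body cosmetic, following the attached parent mesh's pose.
# Additive bone offsets shift the backpack attachment and blend front and back pouches for crouch, tactical sprint, grind rail and skydive.
# Each update samples jaw and under-chin transforms from the head part (stored as "Body Mesh"), replacing those bones when face bones exist.
# A Control Rig runs last, receiving crouch and emote flags.
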AnimBlueprint /BRCosmetics/Characters/Player/Male/Medium/Bodies/M_MED_FalseVulture/Meshes/M_MED_FalseVulture_AnimBP
target_skeleton=&/M_MED_FalseVulture_Skeleton
graphs [
 AnimGraph {nodes [
   n1 = "Transform (Modify) Bone - Bone: attach_backpack"!(ComponentPose=Local To Component(LocalPose=Copy Pose From Mesh() {bUseAttachedParent=true}), Alpha=1, Translation="0,3,0", Rotation="0,0,5") -> (Pose) {
    rotationMode=BMM_Additive
    rotationSpace=BCS_BoneSpace
    scaleSpace=BCS_BoneSpace
    translationMode=BMM_Additive
    translationSpace=BCS_BoneSpace
    alphaBoolBlend {blendInTime=0.2, blendOption=Linear, blendOutTime=0.2}
    boneToModify {boneName=attach_backpack}
    rotation {roll=5}
    translation {y=3}
   }
   n2 = Get bIsCrouching()
   n3 = "Transform (Modify) Bone - Bone: mstr_dyn_pouch_front_01"!(ComponentPose=@n1, bAlphaBoolEnabled:bool=@n2, Translation="-1.384,-6.778,-2.888", Rotation="-8.835,3.744,-80.158", Scale="1,1,1") -> (Pose) {
    alphaInputType=Bool
    rotationMode=BMM_Additive
    rotationSpace=BCS_BoneSpace
    scaleSpace=BCS_BoneSpace
    translationMode=BMM_Additive
    translationSpace=BCS_BoneSpace
    alphaBoolBlend {blendInTime=0.2, blendOption=Linear, blendOutTime=0.2}
    boneToModify {boneName=mstr_dyn_pouch_front_01}
    rotation=<rot -8.8349 3.744 -80.158>
    translation=<-1.384 -6.778 -2.888>
   }
   n4 = "Transform (Modify) Bone - Bone: mstr_dyn_pouch_back_01"!(ComponentPose=@n3, bAlphaBoolEnabled:bool=@n2, Translation="-1.990,-6.649,-4.845", Rotation="-13.229,1.849,-54.260", Scale="1,1,1") -> (Pose) {
    alphaInputType=Bool
    rotationMode=BMM_Additive
    rotationSpace=BCS_BoneSpace
    scaleSpace=BCS_BoneSpace
    translationMode=BMM_Additive
    translationSpace=BCS_BoneSpace
    alphaBoolBlend {blendInTime=0.2, blendOption=Linear, blendOutTime=0.2}
    boneToModify {boneName=mstr_dyn_pouch_back_01}
    rotation=<rot -13.229 1.849 -54.26>
    translation=<-1.99 -6.649 -4.845>
   }
   n5 = Get bIsTacticalSprinting()
   n6 = "Transform (Modify) Bone - Bone: mstr_dyn_pouch_front_01"(ComponentPose=@n4, bAlphaBoolEnabled=@n5, Translation="-1.384,-6.778,-2.888", Rotation="-10.340,5.442,-77.668", Scale="1,1,1") {
    alphaInputType=Bool
    rotationMode=BMM_Additive
    rotationSpace=BCS_BoneSpace
    scaleSpace=BCS_BoneSpace
    translationMode=BMM_Additive
    translationSpace=BCS_BoneSpace
    alphaBoolBlend {blendInTime=0.2, blendOption=Linear, blendOutTime=0.2}
    boneToModify {boneName=mstr_dyn_pouch_front_01}
    rotation=<rot -10.34 5.442 -77.668>
    translation=<-1.384 -6.778 -2.888>
   }
   n7 = "Transform (Modify) Bone - Bone: mstr_dyn_pouch_back_01"(ComponentPose=@n6, bAlphaBoolEnabled=@n5, Translation="-1.990,-6.649,-4.845", Rotation="-11.976,0.077,-53.873", Scale="1,1,1") {
    alphaInputType=Bool
    rotationMode=BMM_Additive
    rotationSpace=BCS_BoneSpace
    scaleSpace=BCS_BoneSpace
    translationMode=BMM_Additive
    translationSpace=BCS_BoneSpace
    alphaBoolBlend {blendInTime=0.2, blendOption=Linear, blendOutTime=0.2}
    boneToModify {boneName=mstr_dyn_pouch_back_01}
    rotation=<rot -11.976 0.0771 -53.873>
    translation=<-1.99 -6.649 -4.845>
   }
   n8 = Get bIsOnGrindRail()
   n9 = "Transform (Modify) Bone - Bone: mstr_dyn_pouch_front_01"(ComponentPose=@n7, bAlphaBoolEnabled=@n8, Translation="-0.517,-5.355,-3.281", Rotation="-10.359,13.439,-81.777", Scale="1,1,1") {
    alphaInputType=Bool
    rotationMode=BMM_Additive
    rotationSpace=BCS_BoneSpace
    scaleSpace=BCS_BoneSpace
    translationMode=BMM_Additive
    translationSpace=BCS_BoneSpace
    alphaBoolBlend {blendInTime=0.2, blendOption=Linear, blendOutTime=0.2}
    boneToModify {boneName=mstr_dyn_pouch_front_01}
    rotation=<rot -10.359 13.439 -81.777>
    translation=<-0.517 -5.355 -3.281>
   }
   n10 = "Transform (Modify) Bone - Bone: mstr_dyn_pouch_back_01"(ComponentPose=@n9, bAlphaBoolEnabled=@n8, Translation="-1.604,-5.464,-5.814", Rotation="-11.277,-0.895,-53.676", Scale="1,1,1") {
    alphaInputType=Bool
    rotationMode=BMM_Additive
    rotationSpace=BCS_BoneSpace
    scaleSpace=BCS_BoneSpace
    translationMode=BMM_Additive
    translationSpace=BCS_BoneSpace
    alphaBoolBlend {blendInTime=0.2, blendOption=Linear, blendOutTime=0.2}
    boneToModify {boneName=mstr_dyn_pouch_back_01}
    rotation=<rot -11.277 -0.895 -53.676>
    translation=<-1.604 -5.464 -5.814>
   }
   n11 = Get bIsSkyDiving()
   n12 = "Transform (Modify) Bone - Bone: mstr_dyn_pouch_front_01"(ComponentPose=@n10, bAlphaBoolEnabled=@n11, Translation="-1.384,-6.778,-2.888", Rotation="-8.602,-0.683,-65.809", Scale="1,1,1") {
    alphaInputType=Bool
    rotationMode=BMM_Additive
    rotationSpace=BCS_BoneSpace
    scaleSpace=BCS_BoneSpace
    translationMode=BMM_Additive
    translationSpace=BCS_BoneSpace
    alphaBoolBlend {blendInTime=0.2, blendOption=Linear, blendOutTime=0.2}
    boneToModify {boneName=mstr_dyn_pouch_front_01}
    rotation=<rot -8.6019 -0.6829 -65.809>
    translation=<-1.384 -6.778 -2.888>
   }
   n13 = "Transform (Modify) Bone - Bone: mstr_dyn_pouch_back_01"(ComponentPose=@n12, bAlphaBoolEnabled=@n11, Translation="-1.990,-6.649,-4.845", Rotation="-8.536,-5.398,-52.664", Scale="1,1,1") {
    alphaInputType=Bool
    rotationMode=BMM_Additive
    rotationSpace=BCS_BoneSpace
    scaleSpace=BCS_BoneSpace
    translationMode=BMM_Additive
    translationSpace=BCS_BoneSpace
    alphaBoolBlend {blendInTime=0.2, blendOption=Linear, blendOutTime=0.2}
    boneToModify {boneName=mstr_dyn_pouch_back_01}
    rotation=<rot -8.536 -5.398 -52.664>
    translation=<-1.99 -6.649 -4.845>
   }
   n14 = "Transform (Modify) Bone - Bone: FACIAL_C_UnderChin"!(
    ComponentPose=@n13
    bAlphaBoolEnabled:bool=Get IsFaceBones() -> (IsFaceBones:bool)
    Translation=Get FACIAL_C_UnderChin_Translation() -> (FACIAL_C_UnderChin_Translation:Vector)
    Rotation=Get FACIAL_C_UnderChin_Rotation() -> (FACIAL_C_UnderChin_Rotation:Rotator)
    Scale="1,1,1"
   ) -> (Pose) {
    alphaInputType=Bool
    rotationMode=BMM_Replace
    rotationSpace=BCS_ParentBoneSpace
    translationMode=BMM_Replace
    translationSpace=BCS_ParentBoneSpace
    alphaBoolBlend {blendOption=Linear}
    boneToModify {boneName=FACIAL_C_UnderChin}
    rotation {roll=20}
   }
   n15 = "Transform (Modify) Bone - Bone: FACIAL_L_UnderChin"!(
    ComponentPose=@n14
    bAlphaBoolEnabled:bool=Get IsFaceBones()
    Translation=Get FACIAL_L_UnderChin_Translation() -> (FACIAL_L_UnderChin_Translation:Vector)
    Rotation=Get FACIAL_L_UnderChin_Rotation() -> (FACIAL_L_UnderChin_Rotation:Rotator)
    Scale="1,1,1"
   ) -> (Pose) {
    alphaInputType=Bool
    rotationMode=BMM_Replace
    rotationSpace=BCS_ParentBoneSpace
    translationMode=BMM_Replace
    translationSpace=BCS_ParentBoneSpace
    alphaBoolBlend {blendOption=Linear}
    boneToModify {boneName=FACIAL_L_UnderChin}
    rotation {roll=20}
   }
   n16 = "Transform (Modify) Bone - Bone: FACIAL_R_UnderChin"!(
    ComponentPose=@n15
    bAlphaBoolEnabled:bool=Get IsFaceBones()
    Translation=Get FACIAL_R_UnderChin_Translation() -> (FACIAL_R_UnderChin_Translation:Vector)
    Rotation=Get FACIAL_R_UnderChin_Rotation() -> (FACIAL_R_UnderChin_Rotation:Rotator)
    Scale="1,1,1"
   ) -> (Pose) {
    alphaInputType=Bool
    rotationMode=BMM_Replace
    rotationSpace=BCS_ParentBoneSpace
    translationMode=BMM_Replace
    translationSpace=BCS_ParentBoneSpace
    alphaBoolBlend {blendOption=Linear}
    boneToModify {boneName=FACIAL_R_UnderChin}
    rotation {roll=20}
   }
   n17 = "Transform (Modify) Bone - Bone: FACIAL_L_Jawline"!(
    ComponentPose=@n16
    bAlphaBoolEnabled:bool=Get IsFaceBones()
    Translation=Get FACIAL_L_Jawline_Translation() -> (FACIAL_L_Jawline_Translation:Vector)
    Rotation=Get FACIAL_L_Jawline_Rotation() -> (FACIAL_L_Jawline_Rotation:Rotator)
    Scale="1,1,1"
   ) -> (Pose) {
    alphaInputType=Bool
    rotationMode=BMM_Replace
    rotationSpace=BCS_ParentBoneSpace
    translationMode=BMM_Replace
    translationSpace=BCS_ParentBoneSpace
    alphaBoolBlend {blendOption=Linear}
    boneToModify {boneName=FACIAL_L_Jawline}
    rotation {roll=20}
   }
   n18 = "Transform (Modify) Bone - Bone: FACIAL_R_Jawline"!(
    ComponentPose=@n17
    bAlphaBoolEnabled:bool=Get IsFaceBones()
    Translation=Get FACIAL_R_Jawline_Translation() -> (FACIAL_R_Jawline_Translation:Vector)
    Rotation=Get FACIAL_R_Jawline_Rotation() -> (FACIAL_R_Jawline_Rotation:Rotator)
    Scale="1,1,1"
   ) -> (Pose) {
    alphaInputType=Bool
    rotationMode=BMM_Replace
    rotationSpace=BCS_ParentBoneSpace
    translationMode=BMM_Replace
    translationSpace=BCS_ParentBoneSpace
    alphaBoolBlend {blendOption=Linear}
    boneToModify {boneName=FACIAL_R_Jawline}
    rotation {roll=20}
   }
   n19 = "Transform (Modify) Bone - Bone: FACIAL_C_Jaw"!(
    ComponentPose=@n18
    bAlphaBoolEnabled:bool=Get IsFaceBones()
    Translation=Get FACIAL_C_Jaw_Tranform() -> (FACIAL_C_Jaw_Tranform:Vector)
    Rotation=Get FACIAL_C_Jaw_Rotation() -> (FACIAL_C_Jaw_Rotation:Rotator)
    Scale="1,1,1"
   ) -> (Pose) {
    alphaInputType=Bool
    rotationMode=BMM_Replace
    rotationSpace=BCS_ParentBoneSpace
    translationMode=BMM_Replace
    translationSpace=BCS_ParentBoneSpace
    alphaBoolBlend {blendOption=Linear}
    boneToModify {boneName=FACIAL_C_Jaw}
    rotation {roll=20}
   }
   n20 = "Transform (Modify) Bone - Bone: FACIAL_C_Jawline"!(
    ComponentPose=@n19
    bAlphaBoolEnabled:bool=Get IsFaceBones()
    Translation=Get FACIAL_C_Jawline_Translation() -> (FACIAL_C_Jawline_Translation:Vector)
    Rotation=Get FACIAL_C_Jawline_Rotation() -> (FACIAL_C_Jawline_Rotation:Rotator)
    Scale="1,1,1"
   ) -> (Pose) {
    alphaInputType=Bool
    rotationMode=BMM_Replace
    rotationSpace=BCS_ParentBoneSpace
    translationMode=BMM_Replace
    translationSpace=BCS_ParentBoneSpace
    alphaBoolBlend {blendOption=Linear}
    boneToModify {boneName=FACIAL_C_Jawline}
    rotation {roll=20}
   }
   n21 = Control Rig!(Source=Component To Local(ComponentPose=@n20), Alpha, IsCrouching:bool=Get bIsCrouching(), bIsPlayingEmote:bool=Get bIsPlayingEmote()) -> (Pose) {
    alphaBoolBlend {blendOption=Linear}
    controlRigAssetReference {blueprintRigClass=&/M_MED_FalseVulture_CtrlRig_New.M_MED_FalseVulture_CtrlRig_New_C}
   }
   Output Pose(Result=@n21) {layerGroup=None}
  ]}
 EventGraph {nodes [
   n1 = Event OnCharacterCustomizationLoaded()
   > Set Body Mesh(Body Mesh:Skeletal Mesh Component Object Reference=Get Array Item with EFort Custom Part Type(Array=@n1.CustomMeshParts, PartType=Head)) -> (Output_Get:Skeletal Mesh Component Object Reference)
   Event BlueprintUpdateAnimation()
   > Is Valid(InputObject=Get Body Mesh() -> (Body Mesh:Skeletal Mesh Component Object Reference))
   n2 = GetSocketBoneName!(self:Skinned Mesh Component Object Reference=Get Body Mesh(), InSocketName:Name=FACIAL_C_UnderChin) -> (ReturnValue:Name=None)
   > n3 = Set IsFaceBones(IsFaceBones:bool="Not Equal Exactly (String)"(A="To String (Name)"(InName=@n2), B=None)) -> (Output_Get:bool)
   > Branch(Condition=@n3)
   n4 = BreakTransform(InTransform=GetSocketTransform(self=Get Body Mesh(), InSocketName=FACIAL_C_UnderChin, TransformSpace=RTS_ParentBoneSpace))
   > Set FACIAL_C_UnderChin_Translation(FACIAL_C_UnderChin_Translation:Vector=@n4) -> (Output_Get:Vector)
   > Set FACIAL_C_UnderChin_Rotation(FACIAL_C_UnderChin_Rotation:Rotator=@n4.Rotation) -> (Output_Get:Rotator)
   n5 = BreakTransform(InTransform=GetSocketTransform(self=Get Body Mesh(), InSocketName=FACIAL_L_UnderChin, TransformSpace=RTS_ParentBoneSpace))
   > Set FACIAL_L_UnderChin_Translation(FACIAL_L_UnderChin_Translation:Vector=@n5) -> (Output_Get:Vector)
   > Set FACIAL_L_UnderChin_Rotation(FACIAL_L_UnderChin_Rotation:Rotator=@n5.Rotation) -> (Output_Get:Rotator)
   n6 = BreakTransform(InTransform=GetSocketTransform(self=Get Body Mesh(), InSocketName=FACIAL_R_UnderChin, TransformSpace=RTS_ParentBoneSpace))
   > Set FACIAL_R_UnderChin_Translation(FACIAL_R_UnderChin_Translation:Vector=@n6) -> (Output_Get:Vector)
   > Set FACIAL_R_UnderChin_Rotation(FACIAL_R_UnderChin_Rotation:Rotator=@n6.Rotation) -> (Output_Get:Rotator)
   n7 = BreakTransform(InTransform=GetSocketTransform(self=Get Body Mesh(), InSocketName=FACIAL_C_Jaw, TransformSpace=RTS_ParentBoneSpace))
   > Set FACIAL_C_Jaw_Tranform(FACIAL_C_Jaw_Tranform:Vector=@n7) -> (Output_Get:Vector)
   > Set FACIAL_C_Jaw_Rotation(FACIAL_C_Jaw_Rotation:Rotator=@n7.Rotation) -> (Output_Get:Rotator)
   n8 = BreakTransform(InTransform=GetSocketTransform(self=Get Body Mesh(), InSocketName=FACIAL_C_Jawline, TransformSpace=RTS_ParentBoneSpace))
   > Set FACIAL_C_Jawline_Translation(FACIAL_C_Jawline_Translation:Vector=@n8) -> (Output_Get:Vector)
   > Set FACIAL_C_Jawline_Rotation(FACIAL_C_Jawline_Rotation:Rotator=@n8.Rotation) -> (Output_Get:Rotator)
   n9 = BreakTransform(InTransform=GetSocketTransform(self=Get Body Mesh(), InSocketName=FACIAL_L_Jawline, TransformSpace=RTS_ParentBoneSpace))
   > Set FACIAL_L_Jawline_Translation(FACIAL_L_Jawline_Translation:Vector=@n9) -> (Output_Get:Vector)
   > Set FACIAL_L_Jawline_Rotation(FACIAL_L_Jawline_Rotation:Rotator=@n9.Rotation) -> (Output_Get:Rotator)
   n10 = BreakTransform(InTransform=GetSocketTransform(self=Get Body Mesh(), InSocketName=FACIAL_R_Jawline, TransformSpace=RTS_ParentBoneSpace))
   > Set FACIAL_R_Jawline_Translation(FACIAL_R_Jawline_Translation:Vector=@n10) -> (Output_Get:Vector)
   > Set FACIAL_R_Jawline_Rotation(FACIAL_R_Jawline_Rotation:Rotator=@n10.Rotation) -> (Output_Get:Rotator)
  ]}
]
variables [
 Body Mesh {type=SkeletalMeshComponent}
 CouchingNotMoving {type=boolean}
 FACIAL_C_Jaw_Rotation {type=Rotator}
 FACIAL_C_Jaw_Tranform {type=Vector}
 FACIAL_C_Jawline_Rotation {type=Rotator}
 FACIAL_C_Jawline_Translation {type=Vector}
 FACIAL_C_UnderChin_Rotation {type=Rotator}
 FACIAL_C_UnderChin_Translation {type=Vector}
 FACIAL_L_Jawline_Rotation {type=Rotator}
 FACIAL_L_Jawline_Translation {type=Vector}
 FACIAL_L_UnderChin_Rotation {type=Rotator}
 FACIAL_L_UnderChin_Translation {type=Vector}
 FACIAL_R_Jawline_Rotation {type=Rotator}
 FACIAL_R_Jawline_Translation {type=Vector}
 FACIAL_R_UnderChin_Rotation {type=Rotator}
 FACIAL_R_UnderChin_Translation {type=Vector}
 IsFaceBones {type=boolean}
 Source Mesh Component {type=SkeletalMeshComponent}
 StandimgNotMoving {type=boolean}
]
