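# Skydive physics asset for the False Vulture medium male face accessory part, covering its dynamic hair bones.
# Back, left and right three-bone capsule chains each attach to a small master sphere through tight swing and twist limits.
# Head, clavicle and upper-spine capsules supply body volumes around the hair.
# Extra constraints with free swing and twist cross-link the left and right tips to the middle and end bones of the back chain.
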PhysicsAsset /BRCosmetics/Characters/Player/Male/Medium/Bodies/M_MED_FalseVulture/Meshes/Parts/M_MED_FalseVulture_FaceAcc_Physics_Skydive
bodies {
 clavicle_l [{
   shapeType=Capsule
   length=26.5
   radius=8.5
   shapeName=clavicle_l_capsule
   center=<9.5566 1.8916 2.6817>
   rotation=<rot -90 179.3666 -179.3676>
  }]
 clavicle_r [{
   shapeType=Capsule
   length=21
   radius=8.5
   shapeName=clavicle_r_capsule
   center=<-6.278 -0.2143 -1.6527>
   rotation=<rot -90 179.9602 -179.9615>
  }]
 dyn_hair_back_01 [{shapeType=Capsule, length=3, radius=2.5, shapeName=dyn_hair_back_01_capsule, center=<3.2251 0 0>}]
 dyn_hair_back_02 [{shapeType=Capsule, length=3, radius=2.5, shapeName=dyn_hair_back_02_capsule, center=<3.2251 0 0>}]
 dyn_hair_back_03 [{shapeType=Capsule, length=10, radius=2.5, shapeName=dyn_hair_back_03_capsule, center=<6.8199 -0.3358 0>}]
 dyn_hair_back_MSTR [{shapeType=Sphere, radius=1, shapeName=dyn_hair_back_MSTR_sphere}]
 dyn_hair_left_01 [{shapeType=Capsule, length=3, radius=2.5, shapeName=dyn_hair_left_01_capsule, center=<3.2251 0 0>, rotation=<rot 0.0005 0 0>}]
 dyn_hair_left_02 [{shapeType=Capsule, length=3, radius=2.5, shapeName=dyn_hair_left_02_capsule, center=<4.1083 0 0.5099>, rotation=<rot 30.0003 0 0>}]
 dyn_hair_left_03 [{
   shapeType=Capsule
   length=3
   radius=2.5
   shapeName=dyn_hair_left_03_capsule
   center=<3.7852 -0.8936 -1.0914>
   rotation=<rot 8.65 -59.619 -64.9626>
  }]
 dyn_hair_left_MSTR [{shapeType=Sphere, radius=1, shapeName=dyn_hair_left_MSTR_sphere}]
 dyn_hair_right_01 [{shapeType=Capsule, length=3.5, radius=2, shapeName=dyn_hair_right_01_capsule, center=<3.2251 0 0>}]
 dyn_hair_right_02 [{shapeType=Capsule, length=6.5, radius=2, shapeName=dyn_hair_right_02_capsule, center=<4.1753 0 -1.9254>}]
 dyn_hair_right_03 [{shapeType=Capsule, length=6.5, radius=2.5, shapeName=dyn_hair_right_03_capsule, center=<3.9054 0 1.0464>, rotation=<rot -75.0005 0 0>}]
 dyn_hair_right_MSTR [{shapeType=Sphere, radius=1, shapeName=dyn_hair_right_MSTR_sphere}]
 head [{shapeType=Capsule, length=16.5, radius=5, shapeName=head_capsule, rotation=<rot -90 -3.2873 3.2868>}]
 spine_05 [{shapeType=Capsule, length=24.5, radius=11, shapeName=spine_05_capsule, center=<1.7379 -1.9039 0>, rotation=<rot 0 -0.0008 0>}]
}
constraints [
 {bone1Name=dyn_hair_back_01, bone2Name=dyn_hair_back_MSTR, swing1LimitDegrees=10, swing2LimitDegrees=20, twistLimitDegrees=2}
 {bone1Name=dyn_hair_back_02, bone2Name=dyn_hair_back_01, swing1LimitDegrees=15, swing2LimitDegrees=10, twistLimitDegrees=5}
 {bone1Name=dyn_hair_back_03, bone2Name=dyn_hair_back_02, swing1LimitDegrees=20, swing2LimitDegrees=20, twistLimitDegrees=5}
 {bone1Name=dyn_hair_left_01, bone2Name=dyn_hair_left_MSTR, swing1LimitDegrees=5, swing2LimitDegrees=15, twistLimitDegrees=5}
 {bone1Name=dyn_hair_left_02, bone2Name=dyn_hair_left_01, swing1LimitDegrees=5, swing2LimitDegrees=10, twistLimitDegrees=5}
 {bone1Name=dyn_hair_left_03, bone2Name=dyn_hair_left_02, swing1LimitDegrees=15, swing2LimitDegrees=10, twistLimitDegrees=5}
 {bone1Name=dyn_hair_right_01, bone2Name=dyn_hair_right_MSTR, swing1LimitDegrees=10, swing2LimitDegrees=15, twistLimitDegrees=2}
 {bone1Name=dyn_hair_right_02, bone2Name=dyn_hair_right_01, swing1LimitDegrees=5, swing2LimitDegrees=20, twistLimitDegrees=2}
 {bone1Name=dyn_hair_right_03, bone2Name=dyn_hair_right_02, swing1LimitDegrees=25, swing2LimitDegrees=25, twistLimitDegrees=5}
 {
  bone1Name=dyn_hair_back_02
  bone2Name=dyn_hair_right_03
  swing1LimitDegrees=45
  swing1Motion=Free
  swing2LimitDegrees=45
  swing2Motion=Free
  twistLimitDegrees=45
  twistMotion=Free
 }
 {
  bone1Name=dyn_hair_back_02
  bone2Name=dyn_hair_left_03
  swing1LimitDegrees=45
  swing1Motion=Free
  swing2LimitDegrees=45
  swing2Motion=Free
  twistLimitDegrees=45
  twistMotion=Free
 }
 {
  bone1Name=dyn_hair_right_03
  bone2Name=dyn_hair_back_03
  swing1LimitDegrees=45
  swing1Motion=Free
  swing2LimitDegrees=45
  swing2Motion=Free
  twistLimitDegrees=45
  twistMotion=Free
 }
 {
  bone1Name=dyn_hair_left_03
  bone2Name=dyn_hair_back_03
  swing1LimitDegrees=45
  swing1Motion=Free
  swing2LimitDegrees=45
  swing2Motion=Free
  twistLimitDegrees=45
  twistMotion=Free
 }
]
