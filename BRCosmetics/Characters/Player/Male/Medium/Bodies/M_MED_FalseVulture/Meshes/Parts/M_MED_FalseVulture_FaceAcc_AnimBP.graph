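# Animation Blueprint for the FalseVulture character's face accessory skeleton, posing and simulating its hair bones on a pose copied from the parent mesh.
# Offset master hair bones feed two neck-relative RigidBody simulations, blended by skydiving state.
# AnimDynamics then sways the front strands, and additive bone offsets reshape side and back hair.
# The EventGraph computes gravity override (via a deprecated function) and head-tilt values, which the AnimGraph does not visibly read.
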AnimBlueprint /BRCosmetics/Characters/Player/Male/Medium/Bodies/M_MED_FalseVulture/Meshes/Parts/M_MED_FalseVulture_FaceAcc_AnimBP
target_skeleton=&/M_MED_FalseVulture_FaceAcc_Skeleton
graphs [
 AnimGraph {nodes [
   n1 = Anim Dynamics - Bone: dyn_hair_front_r_01!(ComponentPose=Local To Component(LocalPose=Use cached pose 'PostRBAN'()), Alpha=1, GravityOverride:Vector="0,0,-980", ExternalForce="0,0,0") -> (Pose) {
    angularDampingOverride=0.9999
    angularSpringConstant=8
    bAngularSpring=true
    bOverrideAngularDamping=true
    bUseGravityOverride=true
    gravityScale=0.35
    simulationSpace=RootRelative
    alphaBoolBlend {blendOption=Linear}
    boundBone {boneName=dyn_hair_front_r_01}
    componentAppliedLinearAccClamp=<600 600 600>
    componentLinearAccScale=<0.2 0.2 0.2>
    componentLinearVelScale=<0.2 0.2 0.2>
    gravityOverride {z=-980}
    physicsBodyDefinitions [{
      sphereCollisionRadius=0
      boundBone {boneName=dyn_hair_front_r_01}
      boxExtents=<8 8 10>
      constraintSetup {angularTargetAxis=AxisY, angularLimitsMax=<3 10 5>, angularLimitsMin=<-3 -20 -25>, angularTarget=<0 -10 0>}
      localJointOffset=<3 0 0>
     }]
   }
   n2 = Anim Dynamics - Bone: dyn_hair_front_l_01!(ComponentPose=@n1, Alpha=1, GravityOverride:Vector="0,0,-980", ExternalForce="0,0,0") -> (Pose) {
    angularDampingOverride=0.9999
    angularSpringConstant=10
    bAngularSpring=true
    bOverrideAngularDamping=true
    bUseGravityOverride=true
    gravityScale=0.35
    simulationSpace=RootRelative
    alphaBoolBlend {blendOption=Linear}
    boundBone {boneName=dyn_hair_front_l_01}
    componentAppliedLinearAccClamp=<600 600 600>
    componentLinearAccScale=<0.25 0.25 0.25>
    componentLinearVelScale=<0.2 0.2 0.2>
    gravityOverride {z=-980}
    physicsBodyDefinitions [{
      sphereCollisionRadius=0
      boundBone {boneName=dyn_hair_front_l_01}
      boxExtents=<7 8 10>
      constraintSetup {angularTargetAxis=AxisY, angularLimitsMax=<5 0 2>, angularLimitsMin=<-5 -20 -25>, angularTarget=<0 -10 0>}
      localJointOffset=<3 0 0>
     }]
   }
   n3 = "Transform (Modify) Bone - Bone: dyn_hair_left_03"(ComponentPose=@n2, Alpha=1, Translation="0,0,2", Rotation="15,0,0", Scale="0.9,1,1") {
    rotationMode=BMM_Additive
    rotationSpace=BCS_BoneSpace
    scaleMode=BMM_Replace
    scaleSpace=BCS_BoneSpace
    translationMode=BMM_Additive
    translationSpace=BCS_BoneSpace
    alphaBoolBlend {blendOption=Linear}
    boneToModify {boneName=dyn_hair_left_03}
    rotation {pitch=15}
    scale {x=0.9}
    translation {z=2}
   }
   n4 = "Transform (Modify) Bone - Bone: dyn_hair_back_02"(ComponentPose=@n3, Alpha=1, Translation="0,1,0", Rotation="0,6,0", Scale="0.9,1,1") {
    rotationMode=BMM_Additive
    rotationSpace=BCS_BoneSpace
    scaleMode=BMM_Replace
    scaleSpace=BCS_BoneSpace
    translationMode=BMM_Additive
    translationSpace=BCS_BoneSpace
    alphaBoolBlend {blendOption=Linear}
    boneToModify {boneName=dyn_hair_back_02}
    rotation {yaw=6}
    scale {x=0.9}
    translation {y=1}
   }
   n5 = "Transform (Modify) Bone - Bone: dyn_hair_back_03"(ComponentPose=@n4, Alpha=1, Translation="0,-2,0", Rotation="0,15,0", Scale="0.9,1,1") {
    rotationMode=BMM_Additive
    rotationSpace=BCS_BoneSpace
    scaleMode=BMM_Replace
    scaleSpace=BCS_BoneSpace
    translationMode=BMM_Additive
    translationSpace=BCS_BoneSpace
    alphaBoolBlend {blendOption=Linear}
    boneToModify {boneName=dyn_hair_back_03}
    rotation {yaw=15}
    scale {x=0.9}
    translation {y=-2}
   }
   n6 = "Transform (Modify) Bone - Bone: dyn_hair_right_02"(ComponentPose=@n5, Alpha=1, Translation="0,-1,0", Rotation="0,0,0", Scale="1,1,1") {
    rotationMode=BMM_Additive
    rotationSpace=BCS_BoneSpace
    scaleMode=BMM_Replace
    scaleSpace=BCS_BoneSpace
    translationMode=BMM_Additive
    translationSpace=BCS_BoneSpace
    alphaBoolBlend {blendOption=Linear}
    boneToModify {boneName=dyn_hair_right_02}
    translation {y=-1}
   }
   n7 = "Transform (Modify) Bone - Bone: dyn_hair_right_03"(ComponentPose=@n6, Alpha=1, Translation="0,1,0", Rotation="-40,45,0", Scale="0.9,1,1") {
    rotationMode=BMM_Additive
    rotationSpace=BCS_BoneSpace
    scaleMode=BMM_Replace
    scaleSpace=BCS_BoneSpace
    translationMode=BMM_Additive
    translationSpace=BCS_BoneSpace
    alphaBoolBlend {blendOption=Linear}
    boneToModify {boneName=dyn_hair_right_03}
    rotation {pitch=-40, yaw=45}
    scale {x=0.9}
    translation {y=1}
   }
   n8 = "Transform (Modify) Bone - Bone: dyn_hair_back_02"(ComponentPose=@n7, Alpha=1, Translation="0,3,0", Rotation="0,0,0", Scale="1,1,1") {
    translationMode=BMM_Additive
    translationSpace=BCS_BoneSpace
    alphaBoolBlend {blendOption=Linear}
    alphaScaleBiasClamp {bInterpResult=true}
    boneToModify {boneName=dyn_hair_back_02}
    translation {y=3}
   }
   n9 = "Transform (Modify) Bone - Bone: dyn_hair_right_02"(ComponentPose=@n8, Alpha=1, Translation="0,0,-3", Rotation="0,0,0", Scale="1,1,1") {
    translationMode=BMM_Additive
    translationSpace=BCS_BoneSpace
    alphaBoolBlend {blendOption=Linear}
    alphaScaleBiasClamp {bInterpResult=true}
    boneToModify {boneName=dyn_hair_right_02}
    translation {z=-3}
   }
   n10 = "Transform (Modify) Bone - Bone: dyn_hair_left_02"(ComponentPose=@n9, Alpha=1, Translation="0,0,3", Rotation="0,0,0", Scale="1,1,1") {
    translationMode=BMM_Additive
    translationSpace=BCS_BoneSpace
    alphaBoolBlend {blendOption=Linear}
    alphaScaleBiasClamp {bInterpResult=true}
    boneToModify {boneName=dyn_hair_left_02}
    translation {z=3}
   }
   n11 = "Transform (Modify) Bone - Bone: dyn_hair_left_01"!(ComponentPose=@n10, bAlphaBoolEnabled:bool=True, Translation="0,-1,0", Rotation="0,0,0", Scale="1,1,1") -> (Pose) {
    alphaInputType=Bool
    translationMode=BMM_Additive
    translationSpace=BCS_BoneSpace
    alphaBoolBlend {blendOption=Linear}
    alphaScaleBiasClamp {bInterpResult=true}
    boneToModify {boneName=dyn_hair_left_01}
    translation {y=-1}
   }
   Output Pose(Result=Component To Local(ComponentPose=@n11)) {layerGroup=None}
   n12 = "Transform (Modify) Bone - Bone: dyn_hair_back_MSTR"(
    ComponentPose=Local To Component(LocalPose=Copy Pose From Mesh() {bUseAttachedParent=true})
    Alpha=1
    Translation="0,0,0"
    Rotation="0,-20,0"
    Scale="1,1,1"
   ) {
    rotationMode=BMM_Additive
    rotationSpace=BCS_BoneSpace
    translationMode=BMM_Additive
    translationSpace=BCS_BoneSpace
    alphaBoolBlend {blendOption=Linear}
    boneToModify {boneName=dyn_hair_back_MSTR}
    rotation {yaw=-20}
   }
   n13 = "Transform (Modify) Bone - Bone: dyn_hair_right_MSTR"(ComponentPose=@n12, Alpha=1, Translation="0,-1,0", Rotation="20.000,0.000,0.000", Scale="1,1,1") {
    rotationMode=BMM_Additive
    rotationSpace=BCS_BoneSpace
    translationMode=BMM_Additive
    translationSpace=BCS_BoneSpace
    alphaBoolBlend {blendOption=Linear}
    boneToModify {boneName=dyn_hair_right_MSTR}
    rotation=<rot 19.9998 0 0>
    translation {y=-1}
   }
   n14 = "Transform (Modify) Bone - Bone: dyn_hair_left_MSTR"(ComponentPose=@n13, Alpha=1, Translation="0,-2,0", Rotation="-20,0,0", Scale="1,1,1") {
    rotationMode=BMM_Additive
    rotationSpace=BCS_BoneSpace
    translationMode=BMM_Additive
    translationSpace=BCS_BoneSpace
    alphaBoolBlend {blendOption=Linear}
    boneToModify {boneName=dyn_hair_left_MSTR}
    rotation {pitch=-20}
    translation {y=-2}
   }
   n15 = "Transform (Modify) Bone - Bone: dyn_hair_back_MSTR"(ComponentPose=@n14, Alpha=1, Translation="0,0,0", Rotation="0,0,0", Scale="1,1,1") {
    rotationMode=BMM_Additive
    rotationSpace=BCS_BoneSpace
    translationMode=BMM_Additive
    translationSpace=BCS_BoneSpace
    alphaBoolBlend {blendOption=Linear}
    boneToModify {boneName=dyn_hair_back_MSTR}
   }
   n16 = "Transform (Modify) Bone - Bone: dyn_hair_right_MSTR"(ComponentPose=@n15, Alpha=1, Translation="-6,-1,5", Rotation="20,-10,0", Scale="1,1,1") {
    rotationMode=BMM_Additive
    rotationSpace=BCS_BoneSpace
    translationMode=BMM_Additive
    translationSpace=BCS_BoneSpace
    alphaBoolBlend {blendOption=Linear}
    boneToModify {boneName=dyn_hair_right_MSTR}
    rotation {pitch=20, yaw=-10}
    translation=<-6 -1 5>
   }
   n17 = "Transform (Modify) Bone - Bone: dyn_hair_left_MSTR"(ComponentPose=@n16, Alpha=1, Translation="0,-1,0", Rotation="-10,0,0", Scale="1,1,1") {
    rotationMode=BMM_Additive
    rotationSpace=BCS_BoneSpace
    translationMode=BMM_Additive
    translationSpace=BCS_BoneSpace
    alphaBoolBlend {blendOption=Linear}
    boneToModify {boneName=dyn_hair_left_MSTR}
    rotation {pitch=-10}
    translation {y=-1}
   }
   Save cached pose 'PreRBANPose'(Pose=Component To Local(ComponentPose=@n17)) {cacheName=PreRBANPose}
   n18 = RigidBody!(ComponentPose=Local To Component(LocalPose=Use cached pose 'PreRBANPose'()), Alpha, OverrideWorldGravity:Vector="0,0,-980", ExternalForce) -> (Pose) {
    bClampLinearTranslationLimitToRefPose=true
    bOverrideWorldGravity=true
    overridePhysicsAsset=&/M_MED_FalseVulture_FaceAcc_Physics_Skydive
    simulationSpace=BaseBoneSpace
    alphaBoolBlend {blendOption=Linear}
    baseBoneRef {boneName=neck_01}
    componentAppliedLinearAccClamp=<900 800 900>
    componentLinearAccScale=<0.5 0.3 0.3>
    componentLinearVelScale=<0.5 0.3 0.3>
    overrideWorldGravity {z=-980}
   }
   n19 = RigidBody(ComponentPose=Local To Component(LocalPose=Use cached pose 'PreRBANPose'()), OverrideWorldGravity="0,0,-980") {
    bClampLinearTranslationLimitToRefPose=true
    bOverrideWorldGravity=true
    overridePhysicsAsset=&/M_MED_FalseVulture_FaceAcc_Physics
    simulationSpace=BaseBoneSpace
    alphaBoolBlend {blendOption=Linear}
    baseBoneRef {boneName=neck_01}
    componentAppliedLinearAccClamp=<900 800 900>
    componentLinearAccScale=<0.5 0.3 0.3>
    componentLinearVelScale=<0.5 0.3 0.3>
    overrideWorldGravity {z=-980}
   }
   Save cached pose 'PostRBAN'(Pose=Blend Poses by bool(BlendPose_0=Component To Local(ComponentPose=@n18), BlendPose_1=Component To Local(ComponentPose=@n19), bActiveValue=Get bIsSkyDiving()) {blendType=Linear}) {cacheName=PostRBAN}
  ]}
 EventGraph {nodes [
   Event BlueprintUpdateAnimation()
   > n1 = Gravity Override State Function DEPRECATED(
    CharacterPartAnimInstance=Self-Reference()
    GameTimeInSeconds=GetGameTimeInSeconds()
    GravityOverrideParams=Get Gravity Override Params()
    RBAN Params=Get RBAN Params()
   )
   > Set Gravity Override(Gravity Override=@n1)
   n2 = GetOwnerSkeletalMeshComponent()
   n3 = BreakRotator(InRot=BreakTransform(InTransform=GetDeltaTransformFromRefPose(self=@n2, BoneName=head, BaseName=spine_05)).Rotation)
   > n4 = Set Head_Back(Head_Back:double=MapRangeClamped(Value=@n3.Yaw, InRangeA=-10, InRangeB=-55)) -> (Output_Get:double)
   > Set Head_Right(Head_Right:double=MapRangeClamped(Value=@n3.Pitch, InRangeA=0, InRangeB=45)) -> (Output_Get:double)
   n5 = OR Boolean(A=Get bIsCrouching(), B=NOT Boolean(A=Get bIsTargeting()))
   n6 = float * float(A=0.3, B=@n4) {is=title K2Node_PromotableOperator}
   n7 = Get bIsMoving2D()
   > Set Head_Back_PoseFix(Head_Back_PoseFix:double=SelectFloat(A=SelectFloat(A=@n6, B=@n4, bPickA=@n5), B=0, bPickA=NOT Boolean(A=@n7))) -> (Output_Get:double)
  ]}
]
variables [
 Gravity Override {type=Vector}
 Gravity Override Params {
  type=GravityOverrideParamsStruct
  default {
   skydive_DivingDown {gravityOverride=<-250 500 -300>, windAmplitude=<50 50 100>}
   skydive_FloatingDown {gravityOverride=<0 -50 -400>, windAmplitude=<50 -50 50>, windFrequency=<1.5 3 2>, windOffset=<0 -50 -50>}
   skydive_GoingUp {gravityOverride=<0 -50 50>, windAmplitude=<100 -50 50>, windOffset=<0 0 50>}
   skydive_Parachuting {windAmplitude=<200 -300 200>}
   tactical_Sprint {gravityOverride=<0 -100 -980>}
  }
 }
 Head_Back {type=number}
 Head_Back_PoseFix {type=number}
 Head_Right {type=number}
 RBAN Params {
  type=RBANParamsStruct
  default {
   dBNO {simSpaceSettings {
     maxAngularAcceleration=1000
     maxAngularVelocity=3
     maxLinearAcceleration=300
     maxLinearVelocity=700
     velocityScaleZ=0
     externalLinearDragV=<0 0.5 0>
    }}
   dBNO_Carried {simSpaceSettings {
     maxAngularAcceleration=1000
     maxAngularVelocity=3
     maxLinearAcceleration=300
     maxLinearVelocity=700
     velocityScaleZ=0
     externalLinearDragV=<0 0.5 0>
    }}
   emoteOrMelee {simSpaceSettings {
     maxAngularAcceleration=1000
     maxAngularVelocity=3
     maxLinearAcceleration=300
     maxLinearVelocity=700
     velocityScaleZ=0
     externalLinearDragV=<0 0.5 0>
    }}
   front_End {simSpaceSettings {
     maxAngularAcceleration=1000
     maxAngularVelocity=3
     maxLinearAcceleration=300
     maxLinearVelocity=700
     velocityScaleZ=0
     externalLinearDragV=<0 0.5 0>
    }}
   onGround_Crouch_Moving {simSpaceSettings {
     maxAngularAcceleration=1000
     maxAngularVelocity=3
     maxLinearAcceleration=300
     maxLinearVelocity=700
     velocityScaleZ=0
     externalLinearDragV=<0 0.5 0>
    }}
   onGround_Falling {simSpaceSettings {
     maxAngularAcceleration=1000
     maxAngularVelocity=3
     maxLinearAcceleration=300
     maxLinearVelocity=700
     velocityScaleZ=0
     externalLinearDragV=<0 0.5 0>
    }}
   onGround_Moving {simSpaceSettings {
     maxAngularAcceleration=1000
     maxAngularVelocity=3
     maxLinearAcceleration=300
     maxLinearVelocity=700
     velocityScaleZ=0
     externalLinearDragV=<0 0.5 0>
    }}
   onGround_Standing {simSpaceSettings {
     maxAngularAcceleration=1000
     maxAngularVelocity=3
     maxLinearAcceleration=300
     maxLinearVelocity=700
     velocityScaleZ=0
     externalLinearDragV=<0 0.5 0>
    }}
   rail_Grinding {simSpaceSettings {
     maxAngularAcceleration=1000
     maxAngularVelocity=3
     maxLinearAcceleration=300
     maxLinearVelocity=700
     velocityScaleZ=0
     worldAlpha=0.1
     externalLinearDragV=<0 0.5 0>
    }}
   riding_Moving {simSpaceSettings {
     maxAngularAcceleration=1000
     maxAngularVelocity=3
     maxLinearAcceleration=300
     maxLinearVelocity=700
     velocityScaleZ=0
     externalLinearDragV=<0 0.5 0>
    }}
   riding_Moving_Targeting {simSpaceSettings {
     maxAngularAcceleration=1000
     maxAngularVelocity=3
     maxLinearAcceleration=300
     maxLinearVelocity=700
     velocityScaleZ=0
     externalLinearDragV=<0 0.5 0>
    }}
   riding_Standing {simSpaceSettings {
     maxAngularAcceleration=1000
     maxAngularVelocity=3
     maxLinearAcceleration=300
     maxLinearVelocity=700
     velocityScaleZ=0
     externalLinearDragV=<0 0.5 0>
    }}
   skydive_DivingDown {simSpaceSettings {
     maxAngularAcceleration=1000
     maxAngularVelocity=3
     maxLinearAcceleration=300
     maxLinearVelocity=700
     velocityScaleZ=0
     externalLinearDragV=<0 0.5 0>
    }}
   skydive_FloatingDown {
    componentLinearAccScale=<0 0.5 0>
    componentLinearVelScale=<0 0.5 0>
    simSpaceSettings {
     maxAngularAcceleration=1000
     maxAngularVelocity=3
     maxLinearAcceleration=300
     maxLinearVelocity=700
     velocityScaleZ=0
     externalLinearDragV=<0 0.5 0>
    }
   }
   skydive_GoingUp {simSpaceSettings {
     maxAngularAcceleration=1000
     maxAngularVelocity=3
     maxLinearAcceleration=300
     maxLinearVelocity=700
     velocityScaleZ=0
     externalLinearDragV=<0 0.5 0>
    }}
   skydive_Parachuting {
    componentLinearAccScale=<0 0.1 0.3>
    componentLinearVelScale=<0 0.1 0.3>
    simSpaceSettings {
     maxAngularAcceleration=1000
     maxAngularVelocity=3
     maxLinearAcceleration=300
     maxLinearVelocity=700
     velocityScaleZ=0
     externalLinearDragV=<0 0.5 0>
    }
   }
   skydive_Surfing {simSpaceSettings {
     maxAngularAcceleration=1000
     maxAngularVelocity=3
     maxLinearAcceleration=300
     maxLinearVelocity=700
     velocityScaleZ=0
     externalLinearDragV=<0 0.5 0>
    }}
   swimming {simSpaceSettings {
     maxAngularAcceleration=1000
     maxAngularVelocity=3
     maxLinearAcceleration=300
     maxLinearVelocity=700
     velocityScaleZ=0
     externalLinearDragV=<0 0.5 0>
    }}
   tactical_Sprint {simSpaceSettings {
     maxAngularAcceleration=1000
     maxAngularVelocity=3
     maxLinearAcceleration=300
     maxLinearVelocity=700
     velocityScaleZ=0
     externalLinearDragV=<0 0.5 0>
    }}
   ziplining {simSpaceSettings {
     maxAngularAcceleration=1000
     maxAngularVelocity=3
     maxLinearAcceleration=300
     maxLinearVelocity=700
     velocityScaleZ=0
     externalLinearDragV=<0 0.5 0>
    }}
  }
 }
]
